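; Season 22 Test Chamber quest minigame actor where players pickaxe lit buttons in sequence to send a quest completion stat.
; The server builds one random sequence for all players, avoiding consecutive repeats; the trigger volume starts eligible players.
; Per-player progress lives in GameplayEffect stacks: correct hits add one, misses reset to the first button, and max stacks finish.
; An input-block effect covers blink transitions, with visuals and sounds limited to the instigating client.
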(event EventBeginPlay
  ;[0 0 0] Begin play - sets up the expected button sequence for all players
  (|CallParentFunction)
  (if (Networking|IsServer)
    (CallFunction|InitButtonSequence)))

(event OnComponentBeginOverlap(GameStartTrigger)
  (params (OverlappedComponent :type Primitive%20Component%20Object%20Reference) (OtherActor :type Actor%20Object%20Reference) (OtherComp :type Primitive%20Component%20Object%20Reference) (OtherBodyIndex :type Integer) (bFromSweep :type Boolean) (SweepResult :type Hit%20Result%20Structure%20%28by%20ref%29))
  ;[0 0 0] GameStartTrigger - checks if the player who enters the volume can play the game, then starts it
  (bind _asfort_player_pawn (Utilities|Casting|CastToFortPlayerPawn OtherActor))
  (bind _bcanplay (CallFunction|CanPlayerPlayGame _asfort_player_pawn))
  (if _bcanplay
    (Networking|SwitchHasAuthority
      (:Authority
        (CallFunction|ApplyGameStartGEsToInstigator _asfort_player_pawn)
        (CallFunction|BlinkAllButtons (Fortnite|GetFortPlayerControllerfromActor _asfort_player_pawn)))
      (:Remote
        (CallFunction|PlaySound (Variables|MinigameComponents|GetGameStart_SoundCue) (Fortnite|GetFortPlayerControllerfromActor _asfort_player_pawn))))
    (Utilities|FlowControl|Delay (Variables|Default|GetBlinkSequenceDuration))
    (CallFunction|SetNextButton (Fortnite|GetFortPlayerControllerfromActor _asfort_player_pawn))))

(event Custom|BlinkAllButtons
  (params (InstigatingPlayerController :type Fort%20Player%20Controller%20Object%20Reference) (bGameFinished :type Boolean))
  ;[0 0 0] Multicast - Blinks all the buttons to indicate the game is starting or complete for the instigating client
  (Utilities|IsValid InstigatingPlayerController
    (:Is%20Valid
      (Networking|SwitchHasAuthority
        (:Authority
          ;;[0.224 0.224 0.224] Input block So the player can't accidentally break their progress while blinking is playing
          (CallFunction|ApplyorRemoveInputBlock true InstigatingPlayerController))
        (:Remote
          ;;[0.2448 0.2448 0.2448] If client - play a timeline between 0 and 1 that changes the material state of the buttons, making them appear to be blinking
          (if bGameFinished
            (CallFunction|PlaySound (Variables|MinigameComponents|GetGameComplete_SoundCue) InstigatingPlayerController))
          (bind (_direction _floattrack) (Timeline|Timeline_1 PlayFromStart)
            (:Update
              (for _array_element (Components|GetChildrenComponents (Variables|Default|GetAllButtons) false)
                (bind _asstatic_mesh_component (Utilities|Casting|CastToStaticMeshComponent _array_element))
                (if (<= _floattrack 0.5)
                  (CallFunction|ChangeButtonState _asstatic_mesh_component 0 InstigatingPlayerController true) ; Default Button State
                  (else
                    (CallFunction|ChangeButtonState _asstatic_mesh_component 2 InstigatingPlayerController true) ; Green Button State
                  ))))))))))

(event Custom|ChangeButtonState
  (params (ButtonStaticMesh :type Static%20Mesh%20Component%20Object%20Reference) (ButtonState :type Integer) (InstigatingPlayerController :type Fort%20Player%20Controller%20Object%20Reference) (bHideWeakpointVFX :type Boolean))
  (bind _returnvalue (Components|GetChildComponent ButtonStaticMesh 0))
  ;[0 0 0] Client only - Changes the material on a given button
  (Utilities|IsValid ButtonStaticMesh
    (:Is%20Valid
      (bind _output (CallFunction|IsValidLocalPlayerController InstigatingPlayerController))
      (if _output
        (switch int ButtonState
          (:0
            (Rendering|Material|SetVectorParameterValueonMaterials ButtonStaticMesh "Emissive" (Variables|Default|GetButtonOff)))
          (:1
            (Rendering|Material|SetVectorParameterValueonMaterials ButtonStaticMesh "Emissive" (Variables|Default|GetButtonRed))
            (CallFunction|PlaySound (Variables|MinigameComponents|GetButtonHighlight_SoundCue) InstigatingPlayerController))
          (:2
            (Rendering|Material|SetVectorParameterValueonMaterials ButtonStaticMesh "Emissive" (Variables|Default|GetButtonGreen))))
        (if bHideWeakpointVFX
          (Components|Activation|Deactivate _returnvalue)
          (else
            (Components|Activation|Activate _returnvalue false)))))))

(event Custom|SetNextButton
  (params (InstigatingPlayerController :type Fort%20Player%20Controller%20Object%20Reference))
  (bind _buttonsequence (Variables|Default|GetButtonSequence))
  ;[0 0 0] Gets the next button in the sequence for the instigating player
  (Utilities|FlowControl|Delay 0.5)
  (Networking|SwitchHasAuthority
    (:Authority
      (CallFunction|ApplyorRemoveInputBlock false InstigatingPlayerController))
    (:Remote
      _))
  (if (Utilities|Array|IsNotEmpty _buttonsequence)
    (CallFunction|ChangeButtonState (Utilities|Array|Get(acopy) _buttonsequence (- (GameplayEffects|GetGameplayEffectCount (Ability|GetFortAbilitySystemComponent (Game|Player|GetPlayerPawn InstigatingPlayerController)) "/BattlePassS22/QuestActors/Narrative_Pt1/TestChamber_Game/GE_Quest_S22_TestChamber_GameCounter.GE_Quest_S22_TestChamber_GameCounter_C" 0 true) 1)) 1 InstigatingPlayerController)))

(event EventOnDamageServer
  (params (Damage :type Float%20%28single-precision%29) (DamageTags :type Gameplay%20Tag%20Container%20Structure%20%28by%20ref%29) (Momentum :type Vector :default "0, 0, 0") (HitInfo :type Hit%20Result%20Structure%20%28by%20ref%29) (InstigatedBy :type Controller%20Object%20Reference) (DamageCauser :type Actor%20Object%20Reference) (EffectContext :type Gameplay%20Effect%20Context%20Handle%20Structure))
  (bind (_bblockinghit _binitialoverlap _time _distance _location _impactpoint _normal _impactnormal _physmat _hitactor _hitcomponent) (Collision|BreakHitResult HitInfo))
  ;[0 0 0] On Damage (server)
  ;;[0.1302 0.1302 0.1302] Check if the player can advance state and if the correct button was hit
  (bind _asfort_player_controller (Utilities|Casting|CastToFortPlayerController InstigatedBy))
  (bind _bcanadvancestate (CallFunction|CanPlayerAdvanceState _asfort_player_controller DamageTags))
  (if _bcanadvancestate
    (bind _asstatic_mesh_component (Utilities|Casting|CastToStaticMeshComponent _hitcomponent))
    (bind (_correctbuttonhit _correctbuttonmesh) (CallFunction|DidPlayerHitCorrectButton _asstatic_mesh_component _asfort_player_controller))
    (if _correctbuttonhit
      ;;[0.1301 0.1301 0.1301] If correct button was hit, move on to the next button in the sequence; else, reset the sequence
      (bind _bgamefinished (CallFunction|AdvanceSequence _asfort_player_controller _asstatic_mesh_component))
      (if _bgamefinished
        (CallFunction|BlinkAllButtons _asfort_player_controller true))
      (else
        (CallFunction|ResetSequence _correctbuttonmesh _asstatic_mesh_component _asfort_player_controller)))))

(event Custom|AdvanceButtonState
  (params (ButtonMesh :type Static%20Mesh%20Component%20Object%20Reference) (InstigatingPlayerController :type Fort%20Player%20Controller%20Object%20Reference) (ColorIndex :type Integer))
  ;[0 0 0] Blink the button the specified color, then advance to the next button in sequence
  (Utilities|IsValid InstigatingPlayerController
    (:Is%20Valid
      (Networking|SwitchHasAuthority
        (:Authority
          (CallFunction|ApplyorRemoveInputBlock true InstigatingPlayerController))
        (:Remote
          _))
      (Utilities|FlowControl|Sequence
        (:then_0
          (Networking|SwitchHasAuthority
            (:Remote
              (CallFunction|PlaySound (Variables|MinigameComponents|GetCorrectButtonHit_SoundCue) InstigatingPlayerController)
              (CallFunction|BlinkButtonTwice ButtonMesh ColorIndex InstigatingPlayerController))))
        (:then_1
          (Utilities|FlowControl|Delay (- (Variables|Default|GetBlinkSequenceDuration) 1.8))
          (CallFunction|SetNextButton InstigatingPlayerController))))))

(event Custom|ResetButtonState
  (params (CorrectButtonMesh :type Static%20Mesh%20Component%20Object%20Reference) (HitButtonMesh :type Static%20Mesh%20Component%20Object%20Reference) (InstigatingPlayerController :type Fort%20Player%20Controller%20Object%20Reference))
  ;[0 0 0] Apply input block to the player and blink the button red to indicate an incorrect hit
  (Utilities|IsValid InstigatingPlayerController
    (:Is%20Valid
      (Networking|SwitchHasAuthority
        (:Authority
          (CallFunction|ApplyorRemoveInputBlock true InstigatingPlayerController))
        (:Remote
          _))
      (Utilities|FlowControl|Sequence
        (:then_0
          (Networking|SwitchHasAuthority
            (:Remote
              (CallFunction|PlaySound (Variables|MinigameComponents|GetWrongButtonHit_SoundCue) InstigatingPlayerController)
              (CallFunction|BlinkButtonTwice HitButtonMesh 1 InstigatingPlayerController)
              (CallFunction|ChangeButtonState CorrectButtonMesh 0 InstigatingPlayerController true))))
        (:then_1
          (Utilities|FlowControl|Delay (- (Variables|Default|GetBlinkSequenceDuration) 1.5))
          (CallFunction|SetNextButton InstigatingPlayerController))))))

(event Custom|PlaySound
  (params (SoundCueToPlay :type Sound%20Cue%20Object%20Reference) (InstigatingPlayerController :type Fort%20Player%20Controller%20Object%20Reference))
  ;[0 0 0] Client only - Plays a passed in sound cue when called
  (bind _output (CallFunction|IsValidLocalPlayerController InstigatingPlayerController))
  (if _output
    (Audio|PlaySoundatLocation SoundCueToPlay (Transformation|GetActorLocation (Game|Player|GetPlayerPawn InstigatingPlayerController)))))

(event Custom|BlinkButtonTwice
  (params (ButtonStaticMesh :type Static%20Mesh%20Component%20Object%20Reference) (ButtonStateToBlink :type Integer) (InstigatingPlayerController :type Fort%20Player%20Controller%20Object%20Reference))
  ;[0 0 0] Client only - helper event for blinking a button the specified color state twice
  (bind _output (CallFunction|IsValidLocalPlayerController InstigatingPlayerController))
  (if _output
    (CallFunction|ChangeButtonState ButtonStaticMesh ButtonStateToBlink InstigatingPlayerController true)
    (Utilities|FlowControl|Delay (Variables|Default|GetBlinkSpeed))
    (CallFunction|ChangeButtonState ButtonStaticMesh 0 InstigatingPlayerController true)
    (Utilities|FlowControl|Delay (Variables|Default|GetBlinkSpeed))
    (CallFunction|ChangeButtonState ButtonStaticMesh ButtonStateToBlink InstigatingPlayerController true)
    (Utilities|FlowControl|Delay (Variables|Default|GetBlinkSpeed))
    (CallFunction|ChangeButtonState ButtonStaticMesh 0 InstigatingPlayerController true)))

(fn AdvanceSequence
  (params (InstigatingPlayerController :type Fort%20Player%20Controller%20Object%20Reference) (ButtonStaticMesh :type Static%20Mesh%20Component%20Object%20Reference))
  ;[0 0 0] Called when the correct button is hit and handles logic for adding a stack of the game counter and advancing the sequence
  (LocalVariables|SetInstigatorControllerLocal InstigatingPlayerController)
  (LocalVariables|SetButtonMesh_Local ButtonStaticMesh)
  (bind _returnvalue (GameplayEffects|ApplyGameplayEffectToTarget :self (Ability|GetAbilitySystemComponent self) :GameplayEffectClass "/BattlePassS22/QuestActors/Narrative_Pt1/TestChamber_Game/GE_Quest_S22_TestChamber_GameCounter.GE_Quest_S22_TestChamber_GameCounter_C" :Target (Ability|GetFortAbilitySystemComponent (LocalVariables|GetInstigatorControllerLocal)) :Level 0.0))
  (bind _returnvalue_2 (Ability|GameplayEffect|GetActiveGameplayEffectStackLimitCount _returnvalue))
  (bind _returnvalue_3 (Ability|GameplayEffect|GetActiveGameplayEffectStackCount _returnvalue))
  (if (!= _returnvalue_3 _returnvalue_2)
    (CallFunction|AdvanceButtonState (LocalVariables|GetButtonMesh_Local) (LocalVariables|GetInstigatorControllerLocal) 2)
    (return false)
    (else
      ;;[0.2031 0.2031 0.2031] Finishes the game when the counter reaches max
      (Development|PrintString "Game finished! " true true "(R=1.000000,G=0.000000,B=0.343179,A=1.000000)" 3.0)
      (GameplayEffects|ApplyGameplayEffectToSelf :self (Ability|GetFortAbilitySystemComponent (LocalVariables|GetInstigatorControllerLocal)) :GameplayEffectClass "/BattlePassS22/QuestActors/Narrative_Pt1/TestChamber_Game/GE_Quest_S22_TestChamber_GameFinished.GE_Quest_S22_TestChamber_GameFinished_C" :Level 0.0)
      (Quest|SendCustomStatEventwithTags :self (Quests|GetQuestManager (LocalVariables|GetInstigatorControllerLocal) "Invalid") :Type "ComplexCustom" :TargetTags "(GameplayTags=((TagName=\"Athena.Quests.S22.TestChamber.GameFinished\")))" :Count 1)
      (return true))))

(fn ApplyGameStartGEstoInstigator
  (params (InstigatingPlayerPawn :type Fort%20Player%20Pawn%20Object%20Reference))
  (bind _returnvalue (Ability|GetFortAbilitySystemComponent InstigatingPlayerPawn))
  ;[0 0 0] Apply GameStart GE and one stack of GameCounter GE
  (GameplayEffects|ApplyGameplayEffectToSelf :self _returnvalue :GameplayEffectClass "/BattlePassS22/QuestActors/Narrative_Pt1/TestChamber_Game/GE_Quest_S22_TestChamber_GameStart.GE_Quest_S22_TestChamber_GameStart_C" :Level 0.0) ; Game Start
  (GameplayEffects|ApplyGameplayEffectToSelf :self _returnvalue :GameplayEffectClass "/BattlePassS22/QuestActors/Narrative_Pt1/TestChamber_Game/GE_Quest_S22_TestChamber_GameCounter.GE_Quest_S22_TestChamber_GameCounter_C" :Level 0.0) ; GameCounter
)

(fn ApplyOrRemoveInputBlock
  (params (bApplyInputBlocker :type Boolean) (InstigatingPlayerController :type Fort%20Player%20Controller%20Object%20Reference))
  (bind _returnvalue (Game|Player|GetPlayerPawn InstigatingPlayerController))
  (bind _returnvalue_2 (Ability|GetFortAbilitySystemComponent _returnvalue))
  ;[0 0 0] Server only - if input value is true, apply the input blocking GE to the instigating player pawn; else, remove it.
  (Networking|SwitchHasAuthority
    (:Authority
      (if bApplyInputBlocker
        (GameplayEffects|ApplyGameplayEffectToSelf :self _returnvalue_2 :GameplayEffectClass "/BattlePassS22/QuestActors/Narrative_Pt1/TestChamber_Game/GE_Quest_S22_TestChamber_InputBlocker.GE_Quest_S22_TestChamber_InputBlocker_C" :Level 0.0) ; Input Block
        (else
          (GameplayEffects|RemoveActiveEffectswithGrantedTags _returnvalue_2 "(GameplayTags=((TagName=\"Granted.Quest.S22.Narrative.TestChamber.InputBlocked\")))")))
      (return))))

(fn CanPlayerAdvanceState
  (params (InstigatingPlayerController :type Fort%20Player%20Controller%20Object%20Reference) (DamageTags :type Gameplay%20Tag%20Container%20Structure))
  (bind _consolidatedquestcomponent (Variables|BuildingGameplayActorPropQuest|GetConsolidatedQuestComponent))
  (bind _returnvalue (Class|ConsolidatedQuestComponent|GetQuestItemDefinition _consolidatedquestcomponent))
  ;[0 0 0] Is the player able to provide input, do they have the quest, and were they using a pickaxe to damage?
  (bind _playerhasquest (Class|AthenaQuestFunctionLibrary|AthenaHasQuest InstigatingPlayerController _returnvalue))
  (bind (_playercontrollerout _completedquestobjective) (Quest|HasPlayerCompletedQuestObjective InstigatingPlayerController _returnvalue "quest_s22_narrative_p01_q14_obj0"))
  (if (or (and _playerhasquest (not _completedquestobjective)) (Variables|Design|QuestSetup|GetbTestVisualsNoQuestRequired))
    (if (and (not (GameplayTags|HasMatchingGameplayTag (Ability|GetFortAbilitySystemComponent (Game|Player|GetPlayerPawn InstigatingPlayerController)) (Variables|Default|GetInputBlockedTag))) (GameplayTags|HasAnyTags DamageTags (Variables|Default|GetPickaxeTag) true))
      (return true)
      (else
        (Development|PrintString "TestChamber: Player is input blocked or damage is not from a pickaxe" true true "(R=1.000000,G=0.000000,B=0.261325,A=1.000000)" 4.0)
        (return false)))
    (else
      (Development|PrintString "TestChamber: Player does not meet quest requirements" true true "(R=1.000000,G=0.000000,B=0.261325,A=1.000000)" 4.0)
      (return false))))

(fn CanPlayerPlayGame
  (params (InstigatingPlayerPawn :type Fort%20Player%20Pawn%20Object%20Reference))
  (bind _consolidatedquestcomponent (Variables|BuildingGameplayActorPropQuest|GetConsolidatedQuestComponent))
  (bind _returnvalue (Class|ConsolidatedQuestComponent|GetQuestItemDefinition _consolidatedquestcomponent))
  (bind _returnvalue_2 (Fortnite|GetFortPlayerControllerfromActor InstigatingPlayerPawn))
  ;[0 0 0] Player can play if they have the corresponding quest AND have not already started an instance of the game
  (bind _playerhasquest (Class|AthenaQuestFunctionLibrary|AthenaHasQuest _returnvalue_2 _returnvalue))
  (bind (_playercontrollerout _completedquestobjective) (Quest|HasPlayerCompletedQuestObjective _returnvalue_2 _returnvalue "quest_s22_narrative_p01_q14_obj0"))
  (return (and (or (and _playerhasquest (not _completedquestobjective)) (Variables|Design|QuestSetup|GetbTestVisualsNoQuestRequired)) (not (GameplayTags|HasMatchingGameplayTag (Ability|GetFortAbilitySystemComponent _returnvalue_2) "(TagName=\"Granted.Quest.S22.Narrative.TestChamber.GameStarted\")")))))

(fn DidPlayerHitCorrectButton
  (params (HitButtonMesh :type Static%20Mesh%20Component%20Object%20Reference) (InstigatingPlayerController :type Fort%20Player%20Controller%20Object%20Reference))
  ;[0 0 0] Did the player hit the correct button in the button sequence, as tracked in their game counter stack?
  (return (== HitButtonMesh (Utilities|Array|Get(acopy) (Variables|Default|GetButtonSequence) (- (GameplayEffects|GetGameplayEffectCount (Ability|GetFortAbilitySystemComponent (Game|Player|GetPlayerPawn InstigatingPlayerController)) "/BattlePassS22/QuestActors/Narrative_Pt1/TestChamber_Game/GE_Quest_S22_TestChamber_GameCounter.GE_Quest_S22_TestChamber_GameCounter_C" 0 true) 1))) (Utilities|Array|Get(acopy) (Variables|Default|GetButtonSequence) (- (GameplayEffects|GetGameplayEffectCount (Ability|GetFortAbilitySystemComponent (Game|Player|GetPlayerPawn InstigatingPlayerController)) "/BattlePassS22/QuestActors/Narrative_Pt1/TestChamber_Game/GE_Quest_S22_TestChamber_GameCounter.GE_Quest_S22_TestChamber_GameCounter_C" 0 true) 1))))

(fn HasGameBeenStarted
  (params (InstigatingPlayerController :type Fort%20Player%20Controller%20Object%20Reference))
  ;[0 0 0] Has an instance of the game been started already for this player?
  (return (>= (GameplayEffects|GetGameplayEffectCount (Ability|GetAbilitySystemComponent (Game|Player|GetPlayerPawn InstigatingPlayerController)) "/BattlePassS22/QuestActors/Narrative_Pt1/TestChamber_Game/GE_Quest_S22_TestChamber_GameStart.GE_Quest_S22_TestChamber_GameStart_C" 0 true) 1)))

(fn InitButtonSequence
  ;[0 0 0] Server only - Creates a random button sequence per the number of iterations. This is the same for all players
  (for _ (range (+ (Variables|Default|GetIterations) 1))
    (LocalVariables|SetButtonsArrayTemp (Components|GetChildrenComponents (Variables|Default|GetAllButtons) false))
    (Utilities|IsValid (LocalVariables|GetButtonPrevious)
      (:Is%20Valid
        (Utilities|Array|RemoveItem (LocalVariables|GetButtonsArrayTemp) (LocalVariables|GetButtonPrevious)))
      (:Is%20Not%20Valid
        _))
    (bind _output_get (LocalVariables|SetButtonPrevious (Utilities|Array|Get(acopy) (LocalVariables|GetButtonsArrayTemp) (Math|Random|RandomIntegerinRange 0 (- (Utilities|Array|Length (LocalVariables|GetButtonsArrayTemp)) 1)))))
    (bind _asstatic_mesh_component (Utilities|Casting|CastToStaticMeshComponent _output_get))
    (Utilities|Array|Add (Variables|Default|GetButtonSequence) _asstatic_mesh_component)
    (Utilities|Array|Clear (LocalVariables|GetButtonsArrayTemp))))

(fn IsValidLocalPlayerController
  (params (InstigatingPlayerController :type Fort%20Player%20Controller%20Object%20Reference))
  ;[0 0 0] Returns true on the executing client if the instigator controller is locally controlled. Returns false if running on the server
  (if (not (Networking|IsServer))
    (Utilities|IsValid InstigatingPlayerController
      (:Is%20Valid
        (if (PlayerController|IsinSplitscreen InstigatingPlayerController)
          (for _array_element (Fortnite|GetLocalFortPlayerControllers true)
            (if (== _array_element InstigatingPlayerController)
              (return true)))
          (else
            _))
        (return true))
      (:Is%20Not%20Valid
        _)))
  (return false))

(fn ResetSequence
  (params (CorrectButtonMesh :type Static%20Mesh%20Component%20Object%20Reference) (HitButtonMesh :type Static%20Mesh%20Component%20Object%20Reference) (InstigatingPlayerController :type Fort%20Player%20Controller%20Object%20Reference))
  (bind _returnvalue (Ability|GetFortAbilitySystemComponent InstigatingPlayerController))
  ;[0 0 0] Called when an incorrect button is hit - resets the GE counter back to 1 so the sequence can restart from the beginning
  (GameplayEffects|RemoveActiveEffectswithTags _returnvalue "(GameplayTags=((TagName=\"Granted.Quest.S22.Narrative.TestChamber.GameCounter\")))")
  (GameplayEffects|ApplyGameplayEffectToSelf :self _returnvalue :GameplayEffectClass "/BattlePassS22/QuestActors/Narrative_Pt1/TestChamber_Game/GE_Quest_S22_TestChamber_GameCounter.GE_Quest_S22_TestChamber_GameCounter_C" :Level 0.0)
  (CallFunction|ResetButtonState CorrectButtonMesh HitButtonMesh InstigatingPlayerController)
  (return))

(fn ConstructionScript
  (bind _buttonmeshoverride (Variables|MinigameComponents|GetButtonMeshOverride))
  (bind _mi_buttondefault (Variables|MinigameComponents|GetMI_ButtonDefault))
  ;[0 0 0] Map variables to the inital button meshes and materials
  (|CallParentFunction)
  (for _array_element (Components|GetChildrenComponents (Variables|Default|GetAllButtons) true)
    (bind _asstatic_mesh_component (Utilities|Casting|CastToStaticMeshComponent _array_element))
    (Utilities|IsValid _buttonmeshoverride
      (:Is%20Valid
        (Components|StaticMesh|SetStaticMesh _asstatic_mesh_component _buttonmeshoverride))
      (:Is%20Not%20Valid
        (Transformation|SetRelativeScale3D _asstatic_mesh_component "0.300000,0.300000,0.200000")))
    (Utilities|IsValid _mi_buttondefault
      (:Is%20Valid
        (Rendering|Material|SetMaterial _asstatic_mesh_component 0 _mi_buttondefault)))))
